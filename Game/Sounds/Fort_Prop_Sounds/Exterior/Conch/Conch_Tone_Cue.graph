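# Sound cue for the conch exterior prop's tone sound, routed through the Athena ambient sound class.
# A Random node picks between two equally weighted tone waves.
# It plays at boosted volume with attenuation borrowed from the Ostrich close-movement settings.
# A Conch concurrency set is assigned, alongside override values of a single instance that prevents new plays.
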SoundCue /Game/Sounds/Fort_Prop_Sounds/Exterior/Conch/Conch_Tone_Cue
root=Random_0
nodes [
 {id=Random_0, children [WavePlayer_1,WavePlayer_2], properties {weights [1,1]}}
 {id=WavePlayer_1, properties {soundWaveAssetPtr=&/Conch_Tone_01}}
 {id=WavePlayer_2, properties {soundWaveAssetPtr=&/Conch_Tone_02}}
]
properties {
 attenuationSettings=/Game/Sounds/Attenuation/Fort_Ostrich_Movement_Close_Att
 maxDistance=6512
 soundClassObject=&Fort_Ambient_Athena
 volumeMultiplier=1.559
 concurrencyOverrides {maxCount=1, resolutionRule=PreventNew}
 concurrencySet [&/Conch_Conc]
}
